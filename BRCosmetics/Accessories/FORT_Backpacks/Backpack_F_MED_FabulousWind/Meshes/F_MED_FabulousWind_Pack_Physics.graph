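# Physics asset for the FabulousWind backpack cosmetic, giving capsule bodies to the dynamic bones of its left and right wings.
# Each mirrored wing has a larger root capsule and seven lettered groups (a–g) of small capsules.
# Angular-limited constraints tie each group to its root; both "a" bodies attach directly, while other groups also link their segments.
# The three-segment "e" chain uses tighter internal swing and twist limits than the rest.
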PhysicsAsset /BRCosmetics/Accessories/FORT_Backpacks/Backpack_F_MED_FabulousWind/Meshes/F_MED_FabulousWind_Pack_Physics
bodies {
 dyn_wing_l_01 [{shapeType=Capsule, length=4, radius=5, shapeName=dyn_wing_l_01_capsule, center=<0.7249 -0.8085 -4.5583>, rotation=<rot 0 93.1806 -66.797>}]
 dyn_wing_l_a_2 [{shapeType=Capsule, length=4, radius=2.5, shapeName=dyn_wing_l_a_2_capsule, rotation=<rot 90 0 0>}]
 dyn_wing_l_a_3 [{shapeType=Capsule, length=4, radius=2.5, shapeName=dyn_wing_l_a_3_capsule, rotation=<rot 90 0 0>}]
 dyn_wing_l_b_2 [{shapeType=Capsule, length=4, radius=2.5, shapeName=dyn_wing_l_b_2_capsule, rotation=<rot 90 0 0>}]
 dyn_wing_l_b_3 [{shapeType=Capsule, length=4, radius=2.5, shapeName=dyn_wing_l_b_3_capsule, rotation=<rot 90 0 0>}]
 dyn_wing_l_c_2 [{shapeType=Capsule, length=4, radius=2.5, shapeName=dyn_wing_l_c_2_capsule, rotation=<rot 90 0 0>}]
 dyn_wing_l_c_3 [{shapeType=Capsule, length=4, radius=2.5, shapeName=dyn_wing_l_c_3_capsule, rotation=<rot 90 0 0>}]
 dyn_wing_l_d_2 [{shapeType=Capsule, length=4, radius=2.5, shapeName=dyn_wing_l_d_2_capsule, rotation=<rot 90 0 0>}]
 dyn_wing_l_d_3 [{shapeType=Capsule, length=4, radius=2.5, shapeName=dyn_wing_l_d_3_capsule, rotation=<rot 90 0 0>}]
 dyn_wing_l_e_2 [{shapeType=Capsule, length=4, radius=2.5, shapeName=dyn_wing_l_e_2_capsule, rotation=<rot 90 0 0>}]
 dyn_wing_l_e_3 [{shapeType=Capsule, length=4, radius=2.5, shapeName=dyn_wing_l_e_3_capsule, rotation=<rot 90 0 0>}]
 dyn_wing_l_e_4 [{shapeType=Capsule, length=4, radius=2.5, shapeName=dyn_wing_l_e_4_capsule, rotation=<rot 90 0 0>}]
 dyn_wing_l_f_2 [{shapeType=Capsule, length=4, radius=2.5, shapeName=dyn_wing_l_f_2_capsule, rotation=<rot 90 0 0>}]
 dyn_wing_l_f_3 [{shapeType=Capsule, length=4, radius=2.5, shapeName=dyn_wing_l_f_3_capsule, rotation=<rot 90 0 0>}]
 dyn_wing_l_g_2 [{shapeType=Capsule, length=4, radius=2.5, shapeName=dyn_wing_l_g_2_capsule, rotation=<rot 90 0 0>}]
 dyn_wing_l_g_3 [{shapeType=Capsule, length=4, radius=2.5, shapeName=dyn_wing_l_g_3_capsule, rotation=<rot 90 0 0>}]
 dyn_wing_r_01 [{shapeType=Capsule, length=4, radius=5, shapeName=dyn_wing_r_01_capsule, center=<-0.4202 0.9326 5.0045>, rotation=<rot 0 94.9525 -65.406>}]
 dyn_wing_r_a_2 [{shapeType=Capsule, length=4, radius=2.5, shapeName=dyn_wing_r_a_2_capsule, rotation=<rot 90 0 0>}]
 dyn_wing_r_a_3 [{shapeType=Capsule, length=4, radius=2.5, shapeName=dyn_wing_r_a_3_capsule, rotation=<rot 90 0 0>}]
 dyn_wing_r_b_2 [{shapeType=Capsule, length=4, radius=2.5, shapeName=dyn_wing_r_b_2_capsule, rotation=<rot 90 0 0>}]
 dyn_wing_r_b_3 [{shapeType=Capsule, length=4, radius=2.5, shapeName=dyn_wing_r_b_3_capsule, rotation=<rot 90 0 0>}]
 dyn_wing_r_c_2 [{shapeType=Capsule, length=4, radius=2.5, shapeName=dyn_wing_r_c_2_capsule, rotation=<rot 90 0 0>}]
 dyn_wing_r_c_3 [{shapeType=Capsule, length=4, radius=2.5, shapeName=dyn_wing_r_c_3_capsule, rotation=<rot 90 0 0>}]
 dyn_wing_r_d_2 [{shapeType=Capsule, length=4, radius=2.5, shapeName=dyn_wing_r_d_2_capsule, rotation=<rot 90 0 0>}]
 dyn_wing_r_d_3 [{shapeType=Capsule, length=4, radius=2.5, shapeName=dyn_wing_r_d_3_capsule, rotation=<rot 90 0 0>}]
 dyn_wing_r_e_2 [{shapeType=Capsule, length=4, radius=2.5, shapeName=dyn_wing_r_e_2_capsule, rotation=<rot 90 0 0>}]
 dyn_wing_r_e_3 [{shapeType=Capsule, length=4, radius=2.5, shapeName=dyn_wing_r_e_3_capsule, rotation=<rot 90 0 0>}]
 dyn_wing_r_e_4 [{shapeType=Capsule, length=4, radius=2.5, shapeName=dyn_wing_r_e_4_capsule, rotation=<rot 90 0 0>}]
 dyn_wing_r_f_2 [{shapeType=Capsule, length=4, radius=2.5, shapeName=dyn_wing_r_f_2_capsule, rotation=<rot 90 0 0>}]
 dyn_wing_r_f_3 [{shapeType=Capsule, length=4, radius=2.5, shapeName=dyn_wing_r_f_3_capsule, rotation=<rot 90 0 0>}]
 dyn_wing_r_g_2 [{shapeType=Capsule, length=4, radius=2.5, shapeName=dyn_wing_r_g_2_capsule, rotation=<rot 90 0 0>}]
 dyn_wing_r_g_3 [{shapeType=Capsule, length=4, radius=2.5, shapeName=dyn_wing_r_g_3_capsule, rotation=<rot 90 0 0>}]
}
constraints [
 {bone1Name=dyn_wing_l_c_2, bone2Name=dyn_wing_l_c_3, swing1LimitDegrees=45, swing2LimitDegrees=45, twistLimitDegrees=45}
 {bone1Name=dyn_wing_l_b_2, bone2Name=dyn_wing_l_b_3, swing1LimitDegrees=45, swing2LimitDegrees=45, twistLimitDegrees=45}
 {bone1Name=dyn_wing_l_d_2, bone2Name=dyn_wing_l_d_3, swing1LimitDegrees=45, swing2LimitDegrees=45, twistLimitDegrees=45}
 {bone1Name=dyn_wing_l_e_4, bone2Name=dyn_wing_l_e_2, swing1LimitDegrees=10, swing2LimitDegrees=15, twistLimitDegrees=15}
 {bone1Name=dyn_wing_l_e_3, bone2Name=dyn_wing_l_e_4, swing1LimitDegrees=25, swing2LimitDegrees=25, twistLimitDegrees=15}
 {bone1Name=dyn_wing_l_f_2, bone2Name=dyn_wing_l_f_3, swing1LimitDegrees=45, swing2LimitDegrees=45, twistLimitDegrees=45}
 {bone1Name=dyn_wing_l_g_2, bone2Name=dyn_wing_l_g_3, swing1LimitDegrees=45, swing2LimitDegrees=45, twistLimitDegrees=45}
 {bone1Name=dyn_wing_r_b_2, bone2Name=dyn_wing_r_b_3, swing1LimitDegrees=45, swing2LimitDegrees=45, twistLimitDegrees=45}
 {bone1Name=dyn_wing_r_c_2, bone2Name=dyn_wing_r_c_3, swing1LimitDegrees=45, swing2LimitDegrees=45, twistLimitDegrees=45}
 {bone1Name=dyn_wing_r_d_2, bone2Name=dyn_wing_r_d_3, swing1LimitDegrees=45, swing2LimitDegrees=45, twistLimitDegrees=45}
 {bone1Name=dyn_wing_r_e_4, bone2Name=dyn_wing_r_e_2, swing1LimitDegrees=10, swing2LimitDegrees=15, twistLimitDegrees=15}
 {bone1Name=dyn_wing_r_e_3, bone2Name=dyn_wing_r_e_4, swing1LimitDegrees=25, swing2LimitDegrees=25, twistLimitDegrees=15}
 {bone1Name=dyn_wing_r_f_2, bone2Name=dyn_wing_r_f_3, swing1LimitDegrees=45, swing2LimitDegrees=45, twistLimitDegrees=45}
 {bone1Name=dyn_wing_r_g_2, bone2Name=dyn_wing_r_g_3, swing1LimitDegrees=45, swing2LimitDegrees=45, twistLimitDegrees=45}
 {bone1Name=dyn_wing_r_a_2, bone2Name=dyn_wing_r_01, swing1LimitDegrees=45, swing2LimitDegrees=45, twistLimitDegrees=45}
 {bone1Name=dyn_wing_r_a_3, bone2Name=dyn_wing_r_01, swing1LimitDegrees=45, swing2LimitDegrees=45, twistLimitDegrees=45}
 {bone1Name=dyn_wing_r_d_3, bone2Name=dyn_wing_r_01, swing1LimitDegrees=45, swing2LimitDegrees=45, twistLimitDegrees=45}
 {bone1Name=dyn_wing_r_e_2, bone2Name=dyn_wing_r_01, swing1LimitDegrees=45, swing2LimitDegrees=45, twistLimitDegrees=45}
 {bone1Name=dyn_wing_r_f_3, bone2Name=dyn_wing_r_01, swing1LimitDegrees=45, swing2LimitDegrees=45, twistLimitDegrees=45}
 {bone1Name=dyn_wing_r_g_3, bone2Name=dyn_wing_r_01, swing1LimitDegrees=45, swing2LimitDegrees=45, twistLimitDegrees=45}
 {bone1Name=dyn_wing_r_c_3, bone2Name=dyn_wing_r_01, swing1LimitDegrees=45, swing2LimitDegrees=45, twistLimitDegrees=45}
 {bone1Name=dyn_wing_r_b_3, bone2Name=dyn_wing_r_01, swing1LimitDegrees=45, swing2LimitDegrees=45, twistLimitDegrees=45}
 {bone1Name=dyn_wing_l_a_2, bone2Name=dyn_wing_l_01, swing1LimitDegrees=45, swing2LimitDegrees=45, twistLimitDegrees=45}
 {bone1Name=dyn_wing_l_a_3, bone2Name=dyn_wing_l_01, swing1LimitDegrees=45, swing2LimitDegrees=45, twistLimitDegrees=45}
 {bone1Name=dyn_wing_l_d_3, bone2Name=dyn_wing_l_01, swing1LimitDegrees=45, swing2LimitDegrees=45, twistLimitDegrees=45}
 {bone1Name=dyn_wing_l_e_2, bone2Name=dyn_wing_l_01, swing1LimitDegrees=45, swing2LimitDegrees=45, twistLimitDegrees=45}
 {bone1Name=dyn_wing_l_f_3, bone2Name=dyn_wing_l_01, swing1LimitDegrees=45, swing2LimitDegrees=45, twistLimitDegrees=45}
 {bone1Name=dyn_wing_l_g_3, bone2Name=dyn_wing_l_01, swing1LimitDegrees=45, swing2LimitDegrees=45, twistLimitDegrees=45}
 {bone1Name=dyn_wing_l_c_3, bone2Name=dyn_wing_l_01, swing1LimitDegrees=45, swing2LimitDegrees=45, twistLimitDegrees=45}
 {bone1Name=dyn_wing_l_b_3, bone2Name=dyn_wing_l_01, swing1LimitDegrees=45, swing2LimitDegrees=45, twistLimitDegrees=45}
]
