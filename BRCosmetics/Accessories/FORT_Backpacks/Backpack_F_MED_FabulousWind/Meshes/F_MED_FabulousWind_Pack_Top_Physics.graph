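# Physics asset for the top mesh of the FabulousWind backpack cosmetic, covering its left and right wing bone chains.
# Each dynamic bone carries one capsule body.
# On each side, constraints chain a master base through two base segments to the main wing bone, which branches into three smaller wing bones.
# Base joints allow only a few degrees of motion, while the branch bones swing more freely, especially on the second swing axis.
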PhysicsAsset /BRCosmetics/Accessories/FORT_Backpacks/Backpack_F_MED_FabulousWind/Meshes/F_MED_FabulousWind_Pack_Top_Physics
bodies {
 dyn_wing_base_l_01 [{
   shapeType=Capsule
   length=6.7108
   radius=2.9598
   shapeName=dyn_wing_base_l_01_capsule
   center=<4.7623 0.0065 0.0528>
   rotation=<rot -1.1575 -94.8971 -82.3159>
  }]
 dyn_wing_base_l_02 [{
   shapeType=Capsule
   length=7.6904
   radius=3.3354
   shapeName=dyn_wing_base_l_02_capsule
   center=<5.4949 -1.5594 -1.5699>
   rotation=<rot 0 100.1796 -70.7103>
  }]
 dyn_wing_base_l_master [{
   shapeType=Capsule
   length=13.4658
   radius=5.6767
   shapeName=dyn_wing_base_l_master_capsule
   center=<-4.2227 0.9024 -6.8861>
   rotation=<rot -11.0362 -90 34.5488>
  }]
 dyn_wing_base_r_01 [{
   shapeType=Capsule
   length=8.1358
   radius=2.8814
   shapeName=dyn_wing_base_r_01_capsule
   center=<-4.7623 0.0065 0.0528>
   rotation=<rot -1.1575 -94.8971 -82.3159>
  }]
 dyn_wing_base_r_02 [{
   shapeType=Capsule
   length=7.7088
   radius=3.6782
   shapeName=dyn_wing_base_r_02_capsule
   center=<-5.6172 1.5838 1.8601>
   rotation=<rot 0 100.2613 -70.9124>
  }]
 dyn_wing_base_r_master [{
   shapeType=Capsule
   length=13.1446
   radius=5.6936
   shapeName=dyn_wing_base_r_master_capsule
   center=<4.4088 -0.9099 7.048>
   rotation=<rot -11.6981 -90 33.6319>
  }]
 dyn_wing_l_01 [{
   shapeType=Capsule
   length=8.9752
   radius=7.6194
   shapeName=dyn_wing_l_01_capsule
   center=<0.7249 -0.8085 -4.5583>
   rotation=<rot 0 93.1806 -66.797>
  }]
 dyn_wing_l_a_1 [{
   shapeType=Capsule
   length=3.5882
   radius=1.1038
   shapeName=dyn_wing_l_a_1_capsule
   center=<4.2092 1.2004 -0.3284>
   rotation=<rot 0 -73.4035 -66.0001>
  }]
 dyn_wing_l_a_2 [{
   shapeType=Capsule
   length=4.3925
   radius=2.3446
   shapeName=dyn_wing_l_a_2_capsule
   center=<2.7958 -0.5119 -2.9949>
   rotation=<rot -9.902 -90 -21.4169>
  }]
 dyn_wing_l_a_3 [{
   shapeType=Capsule
   length=5.7336
   radius=2.9769
   shapeName=dyn_wing_l_a_3_capsule
   center=<-0.935 -0.7708 -3.8558>
   rotation=<rot 5.3128 -90 -43.5498>
  }]
 dyn_wing_r_01 [{
   shapeType=Capsule
   length=8.3892
   radius=7.8416
   shapeName=dyn_wing_r_01_capsule
   center=<-0.4202 0.9326 5.0045>
   rotation=<rot 0 94.9525 -65.406>
  }]
 dyn_wing_r_a_1 [{
   shapeType=Capsule
   length=3.6622
   radius=1.0839
   shapeName=dyn_wing_r_a_1_capsule
   center=<-4.1887 -1.191 0.2943>
   rotation=<rot 0 -73.3622 -65.8731>
  }]
 dyn_wing_r_a_2 [{
   shapeType=Capsule
   length=4.3852
   radius=2.0247
   shapeName=dyn_wing_r_a_2_capsule
   center=<-2.6701 0.383 2.7684>
   rotation=<rot -4.9064 -90 -19.2854>
  }]
 dyn_wing_r_a_3 [{
   shapeType=Capsule
   length=5.7621
   radius=3.1197
   shapeName=dyn_wing_r_a_3_capsule
   center=<0.8867 0.8001 3.7323>
   rotation=<rot 5.7364 -90 -44.0966>
  }]
}
constraints [
 {bone1Name=dyn_wing_base_l_01, bone2Name=dyn_wing_base_l_master, swing1LimitDegrees=5, swing2LimitDegrees=5, twistLimitDegrees=2}
 {bone1Name=dyn_wing_base_l_02, bone2Name=dyn_wing_base_l_01, swing1LimitDegrees=2, swing2LimitDegrees=2, twistLimitDegrees=2}
 {bone1Name=dyn_wing_l_01, bone2Name=dyn_wing_base_l_02, swing1LimitDegrees=5, swing2LimitDegrees=5, twistLimitDegrees=5}
 {bone1Name=dyn_wing_base_r_01, bone2Name=dyn_wing_base_r_master, swing1LimitDegrees=2, swing2LimitDegrees=2, twistLimitDegrees=2}
 {bone1Name=dyn_wing_base_r_02, bone2Name=dyn_wing_base_r_01, swing1LimitDegrees=5, swing2LimitDegrees=5, twistLimitDegrees=2}
 {bone1Name=dyn_wing_r_01, bone2Name=dyn_wing_base_r_02, swing1LimitDegrees=5, swing2LimitDegrees=5, twistLimitDegrees=5}
 {bone1Name=dyn_wing_l_a_1, bone2Name=dyn_wing_l_01, swing1LimitDegrees=10, swing2LimitDegrees=25, twistLimitDegrees=10}
 {bone1Name=dyn_wing_l_a_2, bone2Name=dyn_wing_l_01, swing1LimitDegrees=10, swing2LimitDegrees=25, twistLimitDegrees=10}
 {bone1Name=dyn_wing_l_a_3, bone2Name=dyn_wing_l_01, swing1LimitDegrees=10, swing2LimitDegrees=25, twistLimitDegrees=10}
 {bone1Name=dyn_wing_r_a_1, bone2Name=dyn_wing_r_01, swing1LimitDegrees=10, swing2LimitDegrees=25, twistLimitDegrees=10}
 {bone1Name=dyn_wing_r_a_2, bone2Name=dyn_wing_r_01, swing1LimitDegrees=10, swing2LimitDegrees=25, twistLimitDegrees=10}
 {bone1Name=dyn_wing_r_a_3, bone2Name=dyn_wing_r_01, swing1LimitDegrees=10, swing2LimitDegrees=25, twistLimitDegrees=10}
]
