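# Animation Blueprint for the FabulousWind winged back bling, posing its wings to match the player's movement.
# A wingPos state machine picks poses for targeting, jumping, falling, diving, floating and hanging, adjusting for strafe direction and umbrella gliders.
# Two RigidBody passes add secondary motion before a cached pose receives additive poses driven by sprint speed, crouch and targeting.
# Bone modifiers then offset the pack and rotate the wings during high upward velocity without a parachute.
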AnimBlueprint /BRCosmetics/Accessories/FORT_Backpacks/Backpack_F_MED_FabulousWind/Meshes/F_MED_FabulousWind_Pack_AnimBP
target_skeleton=/Game/Accessories/FORT_Backpacks/Backpack_F_MED_FabulousWind/Meshes/F_MED_FabulousWind_Pack_Skeleton
graphs [
 AnimGraph {nodes [
   n1 = Get SpeedAlpha() -> (SpeedAlpha:float)
   n2 = FabulousWind_Backpack_Additive_PoseAsset : back!(PoseWeight:float=@n1) -> (Pose:Pose Link Structure) {
    is=title AnimGraphNode_PoseByName
    poseAsset=&/FabulousWind_Backpack_Additive_PoseAsset
    poseName=back
   }
   n3 = FabulousWind_Backpack_Additive_PoseAsset : base!(PoseWeight:float=@n1) -> (Pose:Pose Link Structure) {
    is=title AnimGraphNode_PoseByName
    poseAsset=&FabulousWind_Backpack_Additive_PoseAsset
    poseName=base
   }
   n4 = Apply Additive(Base=Apply Additive(Base=Use cached pose 'preSprint'(), Additive=@n2, Alpha=0.75) {alphaBoolBlend {blendOption=Linear}}, Additive=@n3) {alphaBoolBlend {blendOption=Linear}}
   n5 = FabulousWind_Backpack_Additive_PoseAsset : back(PoseWeight=0.05) {poseAsset=&FabulousWind_Backpack_Additive_PoseAsset, poseName=back}
   n6 = Use cached pose 'preSprint'()
   n7 = FabulousWind_Backpack_Additive_PoseAsset : base(PoseWeight=1) {poseAsset=&FabulousWind_Backpack_Additive_PoseAsset, poseName=base}
   n8 = FabulousWind_Backpack_Additive_PoseAsset : back(PoseWeight=1) {poseAsset=&FabulousWind_Backpack_Additive_PoseAsset, poseName=back}
   n9 = Use cached pose 'preSprint'()
   n10 = FabulousWind_Backpack_Additive_PoseAsset : base(PoseWeight=1) {poseAsset=&FabulousWind_Backpack_Additive_PoseAsset, poseName=base}
   n11 = Blend Poses by bool(
    BlendPose_0=Apply Additive(Base=Apply Additive(Base=@n6, Additive=@n5, Alpha=0.75) {alphaBoolBlend {blendOption=Linear}}, Additive=@n7) {alphaBoolBlend {blendOption=Linear}}
    BlendPose_1=Apply Additive(Base=Apply Additive(Base=@n9, Additive=@n8, Alpha=0.75) {alphaBoolBlend {blendOption=Linear}}, Additive=@n10) {alphaBoolBlend {blendOption=Linear}}
    BlendTime_0=0.2
    BlendTime_1=0.2
    bActiveValue=Get bIsTargeting()
   ) {blendType=Linear}
   n12 = Get bIsCrouching()
   n13 = Get bGreaterVel() -> (bGreaterVel:bool)
   n14 = "Transform (Modify) Bone - Bone: dyn_wing_base_l_master"!(
    ComponentPose=Local To Component(LocalPose=Blend Poses by bool(BlendPose_0=@n11, BlendPose_1=@n4, BlendTime_0=0.3, BlendTime_1=0.3, bActiveValue=@n12) {blendType=Linear})
    bAlphaBoolEnabled:bool=@n13
    Translation="0,0,0"
    Rotation="0,-10,0"
    Scale="1,1,1"
   ) -> (Pose) {
    alphaInputType=Bool
    rotationMode=BMM_Additive
    alphaBoolBlend {blendInTime=0.2, blendOption=Linear, blendOutTime=4}
    boneToModify {boneName=dyn_wing_base_l_master}
    rotation {yaw=-10}
   }
   n15 = "Transform (Modify) Bone - Bone: dyn_wing_base_r_master"!(ComponentPose=@n14, bAlphaBoolEnabled:bool=@n13, Translation="0,0,0", Rotation="0,10,0", Scale="1,1,1") -> (Pose) {
    alphaInputType=Bool
    rotationMode=BMM_Additive
    alphaBoolBlend {blendInTime=0.2, blendOption=Linear, blendOutTime=4}
    boneToModify {boneName=dyn_wing_base_r_master}
    rotation {yaw=10}
   }
   n16 = "Transform (Modify) Bone - Bone: pack"(ComponentPose=@n15, Translation="0,1.2,0", Rotation="0,0,0") {
    rotationMode=BMM_Additive
    translationMode=BMM_Additive
    alphaBoolBlend {blendOption=Linear}
    boneToModify {boneName=pack}
    translation {y=1.2}
   }
   Output Pose(Result=Component To Local(ComponentPose=@n16)) {layerGroup=None}
   AnimGraphNode_StateMachine_3 = wingPos() {is=title AnimGraphNode_StateMachine}
   n17 = RigidBody(
    ComponentPose=Local To Component(LocalPose=@AnimGraphNode_StateMachine_3)
    Alpha=Get WingRigidAlpha() -> (WingRigidAlpha:float)
    ExternalForce=Get Dynamic_Up_Force() -> (Dynamic_Up_Force:Vector)
   ) {
    bOverrideWorldGravity=true
    overridePhysicsAsset=&/F_MED_FabulousWind_Pack_Physics
    simulationSpace=BaseBoneSpace
    alphaBoolBlend {blendOption=Linear}
    alphaScaleBiasClamp {bInterpResult=true}
    baseBoneRef {boneName=root}
    componentLinearAccScale=<1 1 1>
    componentLinearVelScale=<1 1 1>
    overrideWorldGravity {z=-980}
    simSpaceSettings {maxAngularVelocity=1000}
   }
   n18 = RigidBody(ComponentPose=@n17, Alpha=0.3) {
    bOverrideWorldGravity=true
    overridePhysicsAsset=&/F_MED_FabulousWind_Pack_Top_Physics
    simulationSpace=BaseBoneSpace
    alphaBoolBlend {blendOption=Linear}
    alphaScaleBiasClamp {bInterpResult=true}
    baseBoneRef {boneName=root}
    componentLinearAccScale=<0.25 0.25 0.5>
    componentLinearVelScale=<0.25 0.25 0.5>
    overrideWorldGravity {z=-980}
    simSpaceSettings {maxAngularVelocity=15, maxLinearAcceleration=1500, maxLinearVelocity=10, worldAlpha=1}
   }
   Save cached pose 'preSprint'(Pose=Component To Local(ComponentPose=@n18)) {cacheName=preSprint}
  ]}
 wingPos {
  entry=Base
  path=AnimGraph.AnimGraphNode_StateMachine_3.wingPos
  ids=id
  nodes [
   AnimStateNode_0 = Base() {is=title AnimStateNode}
   AnimStateNode_1 = Targeting() {is=title AnimStateNode}
   AnimStateNode_2 = Diving() {is=title AnimStateNode}
   AnimStateNode_3 = Floating() {is=title AnimStateNode}
   AnimStateNode_4 = Hanging() {is=title AnimStateNode}
   AnimStateNode_5 = Jumping() {is=title AnimStateNode}
   AnimStateNode_6 = Falling() {is=title AnimStateNode}
   AnimStateTransitionNode_0 = AnimStateTransitionNode()
   AnimStateTransitionNode_1 = AnimStateTransitionNode()
   AnimStateTransitionNode_10 = AnimStateTransitionNode()
   AnimStateTransitionNode_11 = AnimStateTransitionNode()
   AnimStateTransitionNode_12 = AnimStateTransitionNode()
   AnimStateTransitionNode_13 = AnimStateTransitionNode()
   AnimStateTransitionNode_14 = AnimStateTransitionNode()
   AnimStateTransitionNode_16 = AnimStateTransitionNode() {crossfadeDuration=10}
   AnimStateTransitionNode_17 = AnimStateTransitionNode()
   AnimStateTransitionNode_18 = AnimStateTransitionNode()
   AnimStateTransitionNode_2 = AnimStateTransitionNode()
   AnimStateTransitionNode_3 = AnimStateTransitionNode()
   AnimStateTransitionNode_4 = AnimStateTransitionNode()
   AnimStateTransitionNode_5 = AnimStateTransitionNode()
   AnimStateTransitionNode_6 = AnimStateTransitionNode()
   AnimStateTransitionNode_7 = AnimStateTransitionNode()
   AnimStateTransitionNode_8 = AnimStateTransitionNode()
   AnimStateTransitionNode_9 = AnimStateTransitionNode()
  ]
  links [
   {
    dst=Diving
    rule=AnimGraph.AnimGraphNode_StateMachine_3.wingPos.AnimStateTransitionNode_1.AnimationTransitionGraph_0
    src=Base
    via=AnimStateTransitionNode_1
   }
   {
    dst=Falling
    rule=AnimGraph.AnimGraphNode_StateMachine_3.wingPos.AnimStateTransitionNode_18.AnimationTransitionGraph_0
    src=Base
    via=AnimStateTransitionNode_18
   }
   {
    dst=Floating
    rule=AnimGraph.AnimGraphNode_StateMachine_3.wingPos.AnimStateTransitionNode_12.AnimationTransitionGraph_0
    src=Base
    via=AnimStateTransitionNode_12
   }
   {
    dst=Hanging
    rule=AnimGraph.AnimGraphNode_StateMachine_3.wingPos.AnimStateTransitionNode_5.AnimationTransitionGraph_0
    src=Base
    via=AnimStateTransitionNode_5
   }
   {
    dst=Jumping
    rule=AnimGraph.AnimGraphNode_StateMachine_3.wingPos.AnimStateTransitionNode_14.AnimationTransitionGraph_0
    src=Base
    via=AnimStateTransitionNode_14
   }
   {
    dst=Targeting
    rule=AnimGraph.AnimGraphNode_StateMachine_3.wingPos.AnimStateTransitionNode_0.AnimationTransitionGraph_0
    src=Base
    via=AnimStateTransitionNode_0
   }
   {
    dst=Base
    rule=AnimGraph.AnimGraphNode_StateMachine_3.wingPos.AnimStateTransitionNode_6.AnimationTransitionGraph_0
    src=Diving
    via=AnimStateTransitionNode_6
   }
   {
    dst=Floating
    rule=AnimGraph.AnimGraphNode_StateMachine_3.wingPos.AnimStateTransitionNode_2.AnimationTransitionGraph_0
    src=Diving
    via=AnimStateTransitionNode_2
   }
   {
    dst=Hanging
    rule=AnimGraph.AnimGraphNode_StateMachine_3.wingPos.AnimStateTransitionNode_8.AnimationTransitionGraph_0
    src=Diving
    via=AnimStateTransitionNode_8
   }
   {
    dst=Base
    rule=AnimGraph.AnimGraphNode_StateMachine_3.wingPos.AnimStateTransitionNode_17.AnimationTransitionGraph_0
    src=Falling
    via=AnimStateTransitionNode_17
   }
   {
    dst=Base
    rule=AnimGraph.AnimGraphNode_StateMachine_3.wingPos.AnimStateTransitionNode_13.AnimationTransitionGraph_0
    src=Floating
    via=AnimStateTransitionNode_13
   }
   {
    dst=Diving
    rule=AnimGraph.AnimGraphNode_StateMachine_3.wingPos.AnimStateTransitionNode_11.AnimationTransitionGraph_0
    src=Floating
    via=AnimStateTransitionNode_11
   }
   {
    dst=Hanging
    rule=AnimGraph.AnimGraphNode_StateMachine_3.wingPos.AnimStateTransitionNode_3.AnimationTransitionGraph_0
    src=Floating
    via=AnimStateTransitionNode_3
   }
   {
    dst=Base
    rule=AnimGraph.AnimGraphNode_StateMachine_3.wingPos.AnimStateTransitionNode_4.AnimationTransitionGraph_0
    src=Hanging
    via=AnimStateTransitionNode_4
   }
   {
    dst=Diving
    rule=AnimGraph.AnimGraphNode_StateMachine_3.wingPos.AnimStateTransitionNode_9.AnimationTransitionGraph_0
    src=Hanging
    via=AnimStateTransitionNode_9
   }
   {
    dst=Floating
    rule=AnimGraph.AnimGraphNode_StateMachine_3.wingPos.AnimStateTransitionNode_10.AnimationTransitionGraph_0
    src=Hanging
    via=AnimStateTransitionNode_10
   }
   {
    dst=Falling
    rule=AnimGraph.AnimGraphNode_StateMachine_3.wingPos.AnimStateTransitionNode_16.AnimationTransitionGraph_0
    src=Jumping
    via=AnimStateTransitionNode_16
   }
   {
    dst=Base
    rule=AnimGraph.AnimGraphNode_StateMachine_3.wingPos.AnimStateTransitionNode_7.AnimationTransitionGraph_0
    src=Targeting
    via=AnimStateTransitionNode_7
   }
  ]
 }
 Base {
  path=AnimGraph.AnimGraphNode_StateMachine_3.wingPos.AnimStateNode_0.Base
  nodes [Output Animation Pose(Result=FabulousWind_Backpack_PoseAsset : base!(PoseWeight:float=1) -> (Pose:Pose Link Structure) {
     is=title AnimGraphNode_PoseByName
     poseAsset=&/FabulousWind_Backpack_PoseAsset
     poseName=base
    })]
 }
 Targeting {
  path=AnimGraph.AnimGraphNode_StateMachine_3.wingPos.AnimStateNode_1.Targeting
  nodes [Output Animation Pose(Result=FabulousWind_Backpack_PoseAsset : back!(PoseWeight:float=1) -> (Pose:Pose Link Structure) {
     is=title AnimGraphNode_PoseByName
     poseAsset=&FabulousWind_Backpack_PoseAsset
     poseName=back
    })]
 }
 Diving {
  path=AnimGraph.AnimGraphNode_StateMachine_3.wingPos.AnimStateNode_2.Diving
  nodes [
   n1 = "Transform (Modify) Bone - Bone: dyn_wing_base_r_master"!(
    ComponentPose=Local To Component(LocalPose=FabulousWind_Backpack_PoseAsset : fly!(PoseWeight:float=1) -> (Pose:Pose Link Structure) {
      is=title AnimGraphNode_PoseByName
      poseAsset=&FabulousWind_Backpack_PoseAsset
      poseName=fly
     })
    Alpha:float=1
    Translation="0,0,0"
    Rotation="0,25,0"
    Scale="1,1,1"
   ) -> (Pose) {
    rotationMode=BMM_Additive
    translationMode=BMM_Additive
    alphaBoolBlend {blendOption=Linear}
    boneToModify {boneName=dyn_wing_base_r_master}
    rotation {yaw=25}
   }
   n2 = "Transform (Modify) Bone - Bone: dyn_wing_base_l_master"!(ComponentPose=@n1, Alpha:float=1, Translation="0,0,0", Rotation="0,-25,0", Scale="1,1,1") -> (Pose) {
    rotationMode=BMM_Additive
    translationMode=BMM_Additive
    alphaBoolBlend {blendOption=Linear}
    boneToModify {boneName=dyn_wing_base_l_master}
    rotation {yaw=-25}
   }
   Output Animation Pose(Result=Component To Local(ComponentPose=@n2))
  ]
 }
 Floating {
  path=AnimGraph.AnimGraphNode_StateMachine_3.wingPos.AnimStateNode_3.Floating
  nodes [
   n1 = FabulousWind_Backpack_PoseAsset : fly(PoseWeight=1) {poseAsset=&FabulousWind_Backpack_PoseAsset, poseName=fly}
   n2 = FabulousWind_Backpack_PoseAsset : dive!(PoseWeight:float=1) -> (Pose:Pose Link Structure) {
    is=title AnimGraphNode_PoseByName
    poseAsset=&FabulousWind_Backpack_PoseAsset
    poseName=dive
   }
   n3 = Get float_right() -> (float_right:double)
   n4 = "Transform (Modify) Bone - Bone: dyn_wing_base_r_master"(
    ComponentPose=Local To Component(LocalPose=Two Way Blend(A=@n1, B=@n2, Alpha=0.5) {blendNode {alphaBoolBlend {blendOption=Linear}}})
    Alpha=@n3
    Translation="0,0,0"
    Rotation="0,20,0"
    Scale="1,1,1"
   ) {
    rotationMode=BMM_Additive
    translationMode=BMM_Additive
    alphaBoolBlend {blendOption=Linear}
    alphaScaleBiasClamp {bInterpResult=true, interpSpeedDecreasing=5, interpSpeedIncreasing=5}
    boneToModify {boneName=dyn_wing_base_r_master}
    rotation {yaw=20}
   }
   n5 = "Transform (Modify) Bone - Bone: dyn_wing_base_l_master"(ComponentPose=@n4, Alpha=@n3, Translation="0,0,0", Rotation="0,-25,0", Scale="1,1,1") {
    rotationMode=BMM_Additive
    translationMode=BMM_Additive
    alphaBoolBlend {blendOption=Linear}
    alphaScaleBiasClamp {bInterpResult=true, interpSpeedDecreasing=5, interpSpeedIncreasing=5}
    boneToModify {boneName=dyn_wing_base_l_master}
    rotation {yaw=-25}
   }
   n6 = Get float_left() -> (float_left:double)
   n7 = "Transform (Modify) Bone - Bone: dyn_wing_base_l_master"(ComponentPose=@n5, Alpha=@n6, Translation="0,0,0", Rotation="0,-20,0", Scale="1,1,1") {
    rotationMode=BMM_Additive
    translationMode=BMM_Additive
    alphaBoolBlend {blendOption=Linear}
    alphaScaleBiasClamp {bInterpResult=true, interpSpeedDecreasing=5, interpSpeedIncreasing=5}
    boneToModify {boneName=dyn_wing_base_l_master}
    rotation {yaw=-20}
   }
   n8 = "Transform (Modify) Bone - Bone: dyn_wing_base_r_master"(ComponentPose=@n7, Alpha=@n6, Translation="0,0,0", Rotation="0,25,0", Scale="1,1,1") {
    rotationMode=BMM_Additive
    translationMode=BMM_Additive
    alphaBoolBlend {blendOption=Linear}
    alphaScaleBiasClamp {bInterpResult=true, interpSpeedDecreasing=5, interpSpeedIncreasing=5}
    boneToModify {boneName=dyn_wing_base_r_master}
    rotation {yaw=25}
   }
   n9 = "Transform (Modify) Bone - Bone: dyn_wing_base_l_master"(ComponentPose=@n8, Alpha=1, Translation="0,0,0", Rotation="0,-5,-5", Scale="1,1,1") {
    rotationMode=BMM_Additive
    translationMode=BMM_Additive
    alphaBoolBlend {blendOption=Linear}
    alphaScaleBiasClamp {bInterpResult=true, interpSpeedDecreasing=5, interpSpeedIncreasing=5}
    boneToModify {boneName=dyn_wing_base_l_master}
    rotation {roll=-5, yaw=-5}
   }
   n10 = "Transform (Modify) Bone - Bone: dyn_wing_base_r_master"(ComponentPose=@n9, Alpha=1, Translation="0,0,0", Rotation="0,5,-5", Scale="1,1,1") {
    rotationMode=BMM_Additive
    translationMode=BMM_Additive
    alphaBoolBlend {blendOption=Linear}
    alphaScaleBiasClamp {bInterpResult=true, interpSpeedDecreasing=5, interpSpeedIncreasing=5}
    boneToModify {boneName=dyn_wing_base_r_master}
    rotation {roll=-5, yaw=5}
   }
   Output Animation Pose(Result=Component To Local(ComponentPose=@n10))
  ]
 }
 Hanging {
  path=AnimGraph.AnimGraphNode_StateMachine_3.wingPos.AnimStateNode_4.Hanging
  nodes [
   n1 = FabulousWind_Backpack_PoseAsset : base(PoseWeight=1) {poseAsset=&FabulousWind_Backpack_PoseAsset, poseName=base}
   n2 = Two Way Blend(
    A=FabulousWind_Backpack_PoseAsset : dive(PoseWeight=1) {poseAsset=&FabulousWind_Backpack_PoseAsset, poseName=dive}
    B=FabulousWind_Backpack_PoseAsset : fly(PoseWeight=1) {poseAsset=&FabulousWind_Backpack_PoseAsset, poseName=fly}
    Alpha=0.5
   ) {blendNode {alphaBoolBlend {blendOption=Linear}}}
   n3 = "Transform (Modify) Bone - Bone: dyn_wing_r_01"!(
    ComponentPose=Local To Component(LocalPose=Two Way Blend(A=@n2, B=@n1, Alpha=0.6) {blendNode {alphaBoolBlend {blendOption=Linear}}})
    bAlphaBoolEnabled:bool=True
    Translation="0,0,0"
    Rotation="0,0,20"
    Scale="1,1,1"
   ) -> (Pose) {
    alphaInputType=Bool
    rotationMode=BMM_Additive
    translationMode=BMM_Additive
    alphaBoolBlend {blendOption=Linear}
    boneToModify {boneName=dyn_wing_r_01}
    rotation {roll=20}
   }
   n4 = "Transform (Modify) Bone - Bone: dyn_wing_l_01"!(ComponentPose=@n3, bAlphaBoolEnabled:bool=True, Translation="0,0,0", Rotation="0,0,20", Scale="1,1,1") -> (Pose) {
    alphaInputType=Bool
    rotationMode=BMM_Additive
    translationMode=BMM_Additive
    alphaBoolBlend {blendOption=Linear}
    boneToModify {boneName=dyn_wing_l_01}
    rotation {roll=20}
   }
   n5 = Get umbrella_glider() -> (umbrella_glider:bool)
   n6 = "Transform (Modify) Bone - Bone: dyn_wing_r_01"(ComponentPose=@n4, bAlphaBoolEnabled=@n5, Translation="0,0,0", Rotation="0,0,20", Scale="1,1,1") {
    alphaInputType=Bool
    rotationMode=BMM_Additive
    translationMode=BMM_Additive
    alphaBoolBlend {blendOption=Linear}
    boneToModify {boneName=dyn_wing_r_01}
    rotation {roll=20}
   }
   n7 = "Transform (Modify) Bone - Bone: dyn_wing_l_01"(ComponentPose=@n6, bAlphaBoolEnabled=@n5, Translation="0,0,0", Rotation="0,0,20", Scale="1,1,1") {
    alphaInputType=Bool
    rotationMode=BMM_Additive
    translationMode=BMM_Additive
    alphaBoolBlend {blendOption=Linear}
    boneToModify {boneName=dyn_wing_l_01}
    rotation {roll=20}
   }
   Output Animation Pose(Result=Component To Local(ComponentPose=@n7))
  ]
 }
 Jumping {
  path=AnimGraph.AnimGraphNode_StateMachine_3.wingPos.AnimStateNode_5.Jumping
  nodes [
   n1 = "Transform (Modify) Bone - Bone: dyn_wing_base_r_master"(
    ComponentPose=Local To Component(LocalPose=FabulousWind_Backpack_PoseAsset : dive(PoseWeight=0.5) {poseAsset=&FabulousWind_Backpack_PoseAsset, poseName=dive})
    Alpha=1
    Translation="0,0,0"
    Rotation="0,5,0"
    Scale="1,1,1"
   ) {
    rotationMode=BMM_Additive
    translationMode=BMM_Additive
    alphaBoolBlend {blendOption=Linear}
    boneToModify {boneName=dyn_wing_base_r_master}
    rotation {yaw=5}
   }
   n2 = "Transform (Modify) Bone - Bone: dyn_wing_base_l_master"(ComponentPose=@n1, Alpha=1, Translation="0,0,0", Rotation="0,-5,0", Scale="1,1,1") {
    rotationMode=BMM_Additive
    translationMode=BMM_Additive
    alphaBoolBlend {blendOption=Linear}
    boneToModify {boneName=dyn_wing_base_l_master}
    rotation {yaw=-5}
   }
   Output Animation Pose(Result=Component To Local(ComponentPose=@n2))
  ]
 }
 Falling {
  path=AnimGraph.AnimGraphNode_StateMachine_3.wingPos.AnimStateNode_6.Falling
  nodes [
   n1 = "Transform (Modify) Bone - Bone: dyn_wing_base_r_master"(
    ComponentPose=Local To Component(LocalPose=FabulousWind_Backpack_PoseAsset : fly(PoseWeight=0.7) {poseAsset=&FabulousWind_Backpack_PoseAsset, poseName=fly})
    Alpha=1
    Translation="0,0,0"
    Rotation="0,5,0"
    Scale="1,1,1"
   ) {
    rotationMode=BMM_Additive
    translationMode=BMM_Additive
    alphaBoolBlend {blendOption=Linear}
    boneToModify {boneName=dyn_wing_base_r_master}
    rotation {yaw=5}
   }
   n2 = "Transform (Modify) Bone - Bone: dyn_wing_base_l_master"(ComponentPose=@n1, Alpha=1, Translation="0,0,0", Rotation="0,-5,0", Scale="1,1,1") {
    rotationMode=BMM_Additive
    translationMode=BMM_Additive
    alphaBoolBlend {blendOption=Linear}
    boneToModify {boneName=dyn_wing_base_l_master}
    rotation {yaw=-5}
   }
   Output Animation Pose(Result=Component To Local(ComponentPose=@n2))
  ]
 }
 AnimationTransitionGraph_0 {
  path=AnimGraph.AnimGraphNode_StateMachine_3.wingPos.AnimStateTransitionNode_0.AnimationTransitionGraph_0
  nodes [Result(bCanEnterTransition=Get bIsTargeting())]
 }
 AnimationTransitionGraph_0 {
  path=AnimGraph.AnimGraphNode_StateMachine_3.wingPos.AnimStateTransitionNode_1.AnimationTransitionGraph_0
  nodes [Result(bCanEnterTransition=Get bIsSkyDiving())]
 }
 AnimationTransitionGraph_0 {
  path=AnimGraph.AnimGraphNode_StateMachine_3.wingPos.AnimStateTransitionNode_10.AnimationTransitionGraph_0
  nodes [Result(bCanEnterTransition=Get bIsSkydiveFloating())]
 }
 AnimationTransitionGraph_0 {
  path=AnimGraph.AnimGraphNode_StateMachine_3.wingPos.AnimStateTransitionNode_11.AnimationTransitionGraph_0
  nodes [Result(bCanEnterTransition=Get bIsSkydiveDiveMode())]
 }
 AnimationTransitionGraph_0 {
  path=AnimGraph.AnimGraphNode_StateMachine_3.wingPos.AnimStateTransitionNode_12.AnimationTransitionGraph_0
  nodes [Result(bCanEnterTransition=Get bIsSkydiveFloating())]
 }
 AnimationTransitionGraph_0 {
  path=AnimGraph.AnimGraphNode_StateMachine_3.wingPos.AnimStateTransitionNode_13.AnimationTransitionGraph_0
  nodes [Result(bCanEnterTransition=NOT Boolean(A=Get bIsSkyDiving()))]
 }
 AnimationTransitionGraph_0 {
  path=AnimGraph.AnimGraphNode_StateMachine_3.wingPos.AnimStateTransitionNode_14.AnimationTransitionGraph_0
  nodes [Result(bCanEnterTransition=Get bIsJumping())]
 }
 AnimationTransitionGraph_0 {
  path=AnimGraph.AnimGraphNode_StateMachine_3.wingPos.AnimStateTransitionNode_16.AnimationTransitionGraph_0
  nodes [Result(bCanEnterTransition=Get bIsFalling())]
 }
 AnimationTransitionGraph_0 {
  path=AnimGraph.AnimGraphNode_StateMachine_3.wingPos.AnimStateTransitionNode_17.AnimationTransitionGraph_0
  nodes [
   n1 = Get bIsFalling()
   Result(bCanEnterTransition=OR Boolean(A=NOT Boolean(A=@n1), B=Get bIsSkyDiving()))
  ]
 }
 AnimationTransitionGraph_0 {
  path=AnimGraph.AnimGraphNode_StateMachine_3.wingPos.AnimStateTransitionNode_18.AnimationTransitionGraph_0
  nodes [
   n1 = Get bIsSkyDiving()
   Result(bCanEnterTransition=AND Boolean(A=Get bIsFalling(), B=NOT Boolean(A=@n1)))
  ]
 }
 AnimationTransitionGraph_0 {
  path=AnimGraph.AnimGraphNode_StateMachine_3.wingPos.AnimStateTransitionNode_2.AnimationTransitionGraph_0
  nodes [Result(bCanEnterTransition=Get bIsSkydiveFloating())]
 }
 AnimationTransitionGraph_0 {
  path=AnimGraph.AnimGraphNode_StateMachine_3.wingPos.AnimStateTransitionNode_3.AnimationTransitionGraph_0
  nodes [Result(bCanEnterTransition=Get bIsParachuteOpened())]
 }
 AnimationTransitionGraph_0 {
  path=AnimGraph.AnimGraphNode_StateMachine_3.wingPos.AnimStateTransitionNode_4.AnimationTransitionGraph_0
  nodes [Result(bCanEnterTransition=NOT Boolean(A=Get bIsSkyDiving()))]
 }
 AnimationTransitionGraph_0 {
  path=AnimGraph.AnimGraphNode_StateMachine_3.wingPos.AnimStateTransitionNode_5.AnimationTransitionGraph_0
  nodes [Result(bCanEnterTransition=Get bIsParachuteOpened())]
 }
 AnimationTransitionGraph_0 {
  path=AnimGraph.AnimGraphNode_StateMachine_3.wingPos.AnimStateTransitionNode_6.AnimationTransitionGraph_0
  nodes [Result(bCanEnterTransition=NOT Boolean(A=Get bIsSkyDiving()))]
 }
 AnimationTransitionGraph_0 {
  path=AnimGraph.AnimGraphNode_StateMachine_3.wingPos.AnimStateTransitionNode_7.AnimationTransitionGraph_0
  nodes [Result(bCanEnterTransition=NOT Boolean(A=Get bIsTargeting()))]
 }
 AnimationTransitionGraph_0 {
  path=AnimGraph.AnimGraphNode_StateMachine_3.wingPos.AnimStateTransitionNode_8.AnimationTransitionGraph_0
  nodes [Result(bCanEnterTransition=Get bIsParachuteOpened())]
 }
 AnimationTransitionGraph_0 {
  path=AnimGraph.AnimGraphNode_StateMachine_3.wingPos.AnimStateTransitionNode_9.AnimationTransitionGraph_0
  nodes [Result(bCanEnterTransition=Get bIsSkydiveDiveMode())]
 }
 EventGraph {nodes [
   Event BlueprintUpdateAnimation()
   > n1 = Sequence!(execute) -> (then_0, then_1, then_2:Exec, then_3:Exec)
   > Set Dynamic_Up_Force(Dynamic_Up_Force:Vector=vector * vector(A:Vector=Get FallingAnimDynamicsUpForce() -> (FallingAnimDynamicsUpForce:Vector), B:Vector="1.5,1.5,1.5") -> (ReturnValue:Vector) {is=title K2Node_CommutativeAssociativeBinaryOperator}) -> (Output_Get:Vector)
   n2 = Get PawnSpeed2D()
   n3 = "float >= float"(A=@n2, B=400) {is=title K2Node_CallFunction}
   > Set skydive_physics(skydive_physics:double=SelectFloat(A=0.1, B=SelectFloat(A=0.6, B=0.3, bPickA=@n3), bPickA=Get bIsSkydiveDiveMode())) -> (Output_Get:double)
   n4 = Get PawnMovementDirectionAzimuth() -> (PawnMovementDirectionAzimuth:float)
   n5 = float - float(A=MapRangeClamped(Value=@n4, InRangeA=0, InRangeB=-90), B=MapRangeClamped(Value=@n4, InRangeA=-90, InRangeB=-180)) {is=title K2Node_CallFunction}
   n6 = NOT Boolean(A=Get bIsSkydiveDiveMode())
   n7 = NOT Boolean(A=Get bIsParachuteOpened())
   n8 = Get bIsSkyDiving()
   > Set float_left(float_left:double=SelectFloat(A="Absolute (Float)"(A=@n5), B=0, bPickA=AND Boolean!(A=@n8, B=@n6, C:bool=@n7) -> (ReturnValue))) -> (Output_Get:double)
   n9 = Get PawnMovementDirectionAzimuth()
   n10 = float - float(A=MapRangeClamped(Value=@n9, InRangeA=0, InRangeB=90), B=MapRangeClamped(Value=@n9, InRangeA=90, InRangeB=180))
   n11 = NOT Boolean(A=Get bIsSkydiveDiveMode())
   n12 = NOT Boolean(A=Get bIsParachuteOpened())
   n13 = Get bIsSkyDiving()
   > Set float_right(float_right:double=SelectFloat(A="Absolute (Float)"(A=@n10), B=0, bPickA=AND Boolean(A=@n13, B=@n11, C=@n12))) -> (Output_Get:double)
   n14 = Cast To FortPlayerPawn(execute=@n1.then_1, Object=GetOwningActor())
   n15 = Get ParachuteAttachment!(self:Fort Player Pawn Object Reference=@n14) -> (ParachuteAttachment:Fort Player Parachute Object Reference)
   > Is Valid(InputObject=@n15)
   n16 = Get ParachuteItemDef!(self:Fort Player Parachute Object Reference=@n15) -> (ParachuteItemDef:Athena Glider Item Definition Object Reference)
   > Is Valid(InputObject=@n16)
   > Set umbrella_glider(umbrella_glider:bool="Equal (Enum)"(
     A:EFortGliderType Enum=Get GliderType!(self:Athena Glider Item Definition Object Reference=@n16) -> (GliderType:EFortGliderType Enum=Glider)
     B:EFortGliderType Enum=Umbrella
    )) -> (Output_Get:bool)
   Branch(execute=@n1.then_2, Condition=Get bIsSprinting())
   > Set SpeedAlpha(SpeedAlpha:float=MapRangeClamped(Value=@n2, InRangeA=550, InRangeB=820)) -> (Output_Get:float)
   n17 = "Max (Float)"(A=Get PawnVelocityZ(), B=0)
   n18 = Get bIsParachuteOpened()
   n19 = Branch(
    execute=@n1.then_3
    Condition=AND Boolean!(A="float > float"(A=@n17, B=3000), B="Not Equal (Boolean)"(A=@n18, B=true) -> (ReturnValue=~)) -> (ReturnValue)
   )
   > Set bGreaterVel(bGreaterVel:bool=true) -> (Output_Get:bool)
   Set bGreaterVel(execute=@n19.else, bGreaterVel=false)
   TryGetPawnOwner()
  ]}
]
variables [
 bGreaterVel {type=boolean}
 Dynamic_Up_Force {type=Vector}
 FallingAnimDynamicsUpForce {type=Vector}
 float_left {type=number}
 float_right {type=number}
 PlayerVector_FwdBwd {type=number}
 PlayerVector_LR {type=number}
 SimSattingsEnab {type=number}
 skydive_physics {type=number}
 SpeedAlpha {type=number}
 umbrella_glider {type=boolean}
 WingRigidAlpha {type=number, default=0.5}
]
